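; BP_FundingComponent is a Blueprint component for War Effort funding stations, under the MeshNetwork gameplay folder, that relays notifications of added funds.
; Its FundingStation0 custom event receives the funds amount, station number and funding choice index, writing them to the Fortnite shipping log.
; It then broadcasts the same values through the OnFundsAdded event dispatcher, leaving the response to bound listeners.
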(event Custom|FundingStation0
  (params (FundsAdded :type Integer64) (FundingStation :type Integer) (FundingChoiceIndex :type Integer))
  (Fortnite|FortShippingLog self (Utilities|String|ToString(Text) (Utilities|Text|FormatText "NSLOCTEXT(\"[D266071F4970D61C7BFF2C876B3D8839]\", \"1D86082A4208A274E1658AB69F0EA4E3\", \"FundingComponent: FundingStation0 called. FundsAdded: {0} FundingStation: {1} FundingChoiceIndex: {2}\")" (Math|Conversions|ToInteger(Integer64) FundsAdded) FundingStation FundingChoiceIndex)))
  (Default|CallOnFundsAdded FundsAdded FundingStation FundingChoiceIndex))
