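; Parent funding manager for the War Effort event, tracking donation totals for funding stations across mesh network nodes.
; Edge servers forward station donations through the funding component; the root adds them to CurrentFunds and persists totals as mesh metadata, restored after restarts.
; Replicated totals then update station displays, the War Effort mutator and client UI.
; Calendar events map multi-choice stations onto shared fund indices and per-round option data.
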(event Custom|SetSharedUnvaultFunding
  ;[1 1 1] Set new shared unvault numbas based on calendar event
  (bind _eventname (MCPContext|WatchCalendarEvents (Utilities|Array|MakeArray "WEUV01" "WEUV02" "WEUV03"))
    (:OnEventActive
      (if (Utilities|String|EqualExactly(String) _eventname "WEUV01")
        (bind _output_get (Variables|Default|SetSharedUnvaultFunding 46))
        (Development|PrintString (Utilities|String|Append "Shared Unvault Index
" (Utilities|String|ToString(Integer) _output_get)) true true "(R=0.000000,G=0.660000,B=1.000000,A=1.000000)" 1e+02)
        (elif (Utilities|String|EqualExactly(String) _eventname "WEUV02")
          (bind _output_get_2 (Variables|Default|SetSharedUnvaultFunding 48))
          (Development|PrintString (Utilities|String|Append "Shared Unvault Index
" (Utilities|String|ToString(Integer) _output_get_2)) true true "(R=0.000000,G=0.660000,B=1.000000,A=1.000000)" 1e+02)
          (elif (Utilities|String|EqualExactly(String) _eventname "WEUV03")
            (bind _output_get_3 (Variables|Default|SetSharedUnvaultFunding 50))
            (Development|PrintString (Utilities|String|Append "Shared Unvault Index
" (Utilities|String|ToString(Integer) _output_get_3)) true true "(R=0.000000,G=0.660000,B=1.000000,A=1.000000)" 1e+02)))))))

(event Custom|RootSetFunds
  ;[1 1 1] Root node only
  (CallFunction|SetSharedUnvaultFunding)
  (bind _result (Math|Integer|IncrementInt (Variables|Default|GetUpdateRootInt))
    (:%20%20
      (Variables|Default|SetUpdateRootInt _result))))

(event EventBeginPlay
  ;[1 1 1] On Begin play initialize
  (Fortnite|FortShippingLog self "Funding Parent BeginPlay")
  (if (Networking|IsDedicatedServer)
    (bind _nodetype (Class|FortAsyncActionMeshNetworkReady|MeshNetworkReadyAsync)
      (:then
        (if (Variables|Debug|GetDebug_SkipMeshNetwork)
          (CallFunction|MeshNetworkReady "Edge")))
      (:OnReady
        (CallFunction|MeshNetworkReady _nodetype)
        (MeshNetworkSubsystem|BindEventtoOnMeshNodeTypeChanged (GameInstanceSubsystems|GetMeshNetworkSubsystem) (EventDispatchers|CreateEvent :!event Custom|MeshNetworkReady))))
    (else
      (Class|FortAsyncActionMeshNetworkReady|MeshNetworkReadyAsync
        (:OnReady
          (Utilities|FlowControl|Delay 3.0)
          (Default|GetFundingStations))))))

(event Custom|UpdateFunds
  (params (Funds :type Integer64) (Funding%20Station :type Integer) (FundingChoiceIndex :type Integer))
  (bind _returnvalue (< FundingChoiceIndex 2))
  (bind _newtotalfunds (Variables|Default|GetNewTotalFunds))
  (bind _newtotalfunds_2 (Variables|Default|GetNewTotalFunds))
  (bind _newtotalfunds_3 (Variables|Default|GetNewTotalFunds))
  (bind (_round1optiondata_option1 _round1optiondata_option2) (Utilities|Struct|BreakWarEffortFundingChoiceData (Variables|GetRound1OptionData_Option1)))
  (bind (_round2optiondata_option1 _round2optiondata_option2) (Utilities|Struct|BreakWarEffortFundingChoiceData (Variables|GetRound2OptionData_Option1)))
  (bind (_round3optiondata_option1 _round3optiondata_option2) (Utilities|Struct|BreakWarEffortFundingChoiceData (Variables|GetRound3OptionData_Option1)))
  (bind (_round4optiondata_option1 _round4optiondata_option2) (Utilities|Struct|BreakWarEffortFundingChoiceData (Variables|GetRound4OptionData_Option1)))
  (bind _newtotalfunds_4 (Variables|Default|GetNewTotalFunds))
  ;[1 1 1] Root node only
  (bind _output_get (Variables|Default|SetInitialFundingStation Funding%20Station))
  (CallFunction|SetSharedUnvaultFunding)
  (bind (_outfundingstation _outfunds _ischoicefund _activecalendarevent) (CallFunction|MultiChoiceCheck Funds _output_get FundingChoiceIndex))
  (Variables|Default|SetFunding%20Station _outfundingstation)
  (bind _output_get_2 (Variables|Default|SetNewTotalFunds (+ _outfunds (Utilities|Array|Get(acopy) (Variables|Default|GetCurrentFunds) (Variables|Default|GetFunding%20Station)))))
  (Utilities|FlowControl|Sequence
    (:then_0
      ;;[0.3424 0.9602 1] If this is a funding choice donation, update the choice data based on the calendar event & choice index
      (if _ischoicefund
        (Fortnite|FortShippingLog self (Utilities|String|ToString(Text) (Utilities|Text|FormatText "NSLOCTEXT(\"[8FA25645435DB4DB77D0CFB0882B521F]\", \"B4B499FC418846262EEB7787E720FAF4\", \"Updating choice data based on calendar event \\'{0}\\'\")" _activecalendarevent)))
        (switch int (Utilities|Array|FindItem (Utilities|Array|MakeArray "WEUV01" "WENW01" "WEUV02" "WEUV03") _activecalendarevent)
          (:Default
            (Fortnite|FortShippingLog self (Utilities|String|ToString(Text) (Utilities|Text|FormatText "NSLOCTEXT(\"[8FA25645435DB4DB77D0CFB0882B521F]\", \"90121E9E472C84EA11109585DB11AFD9\", \"FundingManager: Unknown calendar event: \\'{0}\\'\")" _activecalendarevent))))
          (:0
            (if _returnvalue
              (Utilities|Struct|SetmembersinWarEffortFundingOptionData _round1optiondata_option1 _newtotalfunds_3)
              (else
                (Utilities|Struct|SetmembersinWarEffortFundingOptionData _round1optiondata_option2 _newtotalfunds_3))))
          (:1
            (if _returnvalue
              (Utilities|Struct|SetmembersinWarEffortFundingOptionData _round2optiondata_option1 _newtotalfunds_4)
              (else
                (Utilities|Struct|SetmembersinWarEffortFundingOptionData _round2optiondata_option2 _newtotalfunds_4))))
          (:2
            (if _returnvalue
              (Utilities|Struct|SetmembersinWarEffortFundingOptionData _round3optiondata_option1 _newtotalfunds)
              (else
                (Utilities|Struct|SetmembersinWarEffortFundingOptionData _round3optiondata_option2 _newtotalfunds))))
          (:3
            (if _returnvalue
              (Utilities|Struct|SetmembersinWarEffortFundingOptionData _round4optiondata_option1 _newtotalfunds_2)
              (else
                (Utilities|Struct|SetmembersinWarEffortFundingOptionData _round4optiondata_option2 _newtotalfunds_2)))))))
    (:then_1
      (CallFunction|SetAllFundsArray _output_get_2)
      (Variables|TotalFunds|SetTotalFunds _output_get_2)
      ;[1 1 1] Persist the current vote to the mesh service
      (bind _structout (Utilities|Struct|SetmembersinWarEffortFundingMetadata (Variables|Default|GetFundingStationMetaData) (Utilities|Struct|MakeWarEffortIndexedFundingData (Variables|Default|GetCurrentFunds) (Math|Float|TruncatetoInteger64 (ScalableFloat|GetValueatLevel (Variables|Default|GetFinalFundingLimit))) (Math|Float|TruncatetoInteger64 (ScalableFloat|GetValueatLevel (Variables|Default|GetTowerFundingLimit)))) (Utilities|Array|MakeArray (Variables|Default|GetRound1OptionData) (Variables|Default|GetRound2OptionData) (Variables|Default|GetRound3OptionData) (Variables|Default|GetRound4OptionData))))
      (MeshNetworkSubsystem|SetMetaDatawithKey (GameInstanceSubsystems|GetMeshNetworkSubsystem) (Variables|Default|GetMetaDataKey) _structout))))

(event Custom|MeshMetaDataUpdated
  (bind _indexedfundingdata (Utilities|Struct|BreakWarEffortFundingMetadata (Variables|Default|GetFundingStationMetaData)))
  (bind _indexedfundingdata_currentfundingarray (Utilities|Struct|BreakWarEffortIndexedFundingData _indexedfundingdata))
  ;[1 1 1] Restore initial vote value if the root was restarted
  (Fortnite|FortShippingLog self "FundingManager MeshMetaDataUpdated")
  (bind _returnvalue (MeshNetworkSubsystem|GetMetaDatawithKey (GameInstanceSubsystems|GetMeshNetworkSubsystem) (Variables|Default|GetMetaDataKey) (Variables|Default|GetFundingStationMetaData)))
  (if _returnvalue
    (Variables|Default|SetCurrentFunds _indexedfundingdata_currentfundingarray)
    (CallFunction|UpdateMutatorFundingData)
    (CallFunction|RootSetFunds)
    (else
      (CallFunction|RootSetFunds))))

(event Custom|MeshNetworkReady
  (params (NodeType :type EMeshNetworkNodeType%20Enum :default "Root"))
  ;[1 1 1] Root / Edge Nodes
  (Fortnite|FortShippingLog self (Utilities|String|ToString(Text) (Utilities|Text|FormatText "NSLOCTEXT(\"[8FA25645435DB4DB77D0CFB0882B521F]\", \"6B57088D4EF61F1404A2CF89E7F1CB9D\", \"MeshNetworkReady: NodeType{0}\")" (Utilities|String|EnumtoString NodeType))))
  (switch Utilities|FlowControl|Switch|SwitchonEMeshNetworkNodeType NodeType
    (:Root
      (Default|GetFundingStations)
      (Default|BindEventtoOnFundsAdded (Variables|Default|GetBP_FundingComponent) (EventDispatchers|CreateEvent :!event Custom|UpdateFunds))
      (MeshNetworkSubsystem|BindEventtoOnMeshMetaDataUpdated (GameInstanceSubsystems|GetMeshNetworkSubsystem) (EventDispatchers|CreateEvent :!event Custom|MeshMetaDataUpdated))
      (CallFunction|MeshMetaDataUpdated))
    (:Edge
      (Class|FortAsyncActionAuthorityMatchReady|MatchReadyAsync
        (:OnReady
          (Default|GetFundingStations)
          (for (_array_element _array_index) (Variables|Default|GetFundingPropService)
            (Utilities|IsValid (Utilities|ResolveSoftReference _array_element)
              (:Is%20Valid
                (Fortnite|FortShippingLog self (Utilities|String|ToString(Text) (Utilities|Text|FormatText "NSLOCTEXT(\"[8FA25645435DB4DB77D0CFB0882B521F]\", \"CDC646264AB2B99413EAD7BCF4D2BE8C\", \"FundingManager: Binding FundingRecieved to station {0} (station no. {1})\\'s OnFundingReceived\")" (Utilities|String|ToString(Object) (Utilities|ResolveSoftReference _array_element)) (Variables|/CorruptionGameplay/Gameplay/WarEffort/Prototype/BP_StationProp_FundingStation.BP_StationProp_FundingStation_C|GetFundingStationNum (Utilities|ResolveSoftReference _array_element)))))
                (Default|BindEventtoOnFundingReceived (Utilities|ResolveSoftReference _array_element) (EventDispatchers|CreateEvent :!event Custom|FundingRecieved)))
              (:Is%20Not%20Valid
                (Fortnite|FortShippingLog self (Utilities|String|ToString(Text) (Utilities|Text|FormatText "NSLOCTEXT(\"[8FA25645435DB4DB77D0CFB0882B521F]\", \"80457FD04FED73304A9FC39AEE27B04C\", \"FundingManagerError: Could not resolve station at index {0}!\")" _array_index))))))
          (CallFunction|SetSharedUnvaultFunding)
          (CallFunction|UpdateAllFundingStations)
          (CallFunction|UpdateMutatorFundingData))))))

(event Custom|FundingRecieved
  (params (FundingAmountAdded :type Integer64) (FundingStationNum :type Integer) (FundingChoiceIndex :type Integer))
  ;[0.9792 0.0204 0.0204] On Funds Recieved // KILL ME
  (Fortnite|FortShippingLog self (Utilities|String|ToString(Text) (Utilities|Text|FormatText "NSLOCTEXT(\"[8FA25645435DB4DB77D0CFB0882B521F]\", \"AA90828B4D2C4968BE9B6993031F3C90\", \"FundingManager: FundingRecieved: Amount Added: {0} Funding Station Num: {1} FundingChoiceIndex: {2}\")" (Math|Conversions|ToInteger(Integer64) FundingAmountAdded) FundingStationNum FundingChoiceIndex)))
  (Variables|Default|SetFunding%20Amount%20Added FundingAmountAdded)
  (Variables|Default|SetFunding%20Station%20Num FundingStationNum)
  (Variables|Default|SetFunding%20Choice%20Index FundingChoiceIndex)
  (Utilities|FlowControl|Sequence
    (:then_0
      ;;[1 1 1] This is better
      (Class|BPFundingComponent|FundingStation0 (Variables|Default|GetBP_FundingComponent) (Variables|Default|GetFunding%20Amount%20Added) (Variables|Default|GetFunding%20Station%20Num) (Variables|Default|GetFunding%20Choice%20Index)))
    (:then_1
      ;;[0.3802 0.3353 0] Only runs if in editor/PIE and debug true
      (CallFunction|DebugSkipMeshNetworkUpdateFunds))))

(fn Debug_SkipMeshNetwork_UpdateFunds
  (if (and (Variables|Debug|GetDebug_SkipMeshNetwork) (Editor|InEditororPIE))
    (CallFunction|UpdateFunds (Variables|Default|GetFunding%20Amount%20Added) (Variables|Default|GetFunding%20Station%20Num) (Variables|Default|GetFunding%20Choice%20Index))
    (Utilities|Array|SetArrayElem (Variables|Default|GetCurrentFunds) (Variables|Default|GetFunding%20Station) (Variables|Default|GetNewTotalFunds))))

(fn GetFundingStations
  (Utilities|Array|Clear (Variables|Default|GetFundingPropService))
  (bind _outactors (Actor|GetAllActorsOfClass "/CorruptionGameplay/Gameplay/WarEffort/Prototype/BP_StationProp_FundingStation.BP_StationProp_FundingStation_C"))
  (for _array_element _outactors
    (Variables|/CorruptionGameplay/Gameplay/WarEffort/Prototype/BP_StationProp_FundingStation.BP_StationProp_FundingStation_C|SetFundingManagerRef self _array_element)
    (Utilities|Array|SetArrayElem (Variables|Default|GetFundingPropService) (Variables|/CorruptionGameplay/Gameplay/WarEffort/Prototype/BP_StationProp_FundingStation.BP_StationProp_FundingStation_C|GetFundingStationNum _array_element) (Utilities|ResolveSoftReference _array_element) true)
    ;[1 1 1] Ensure we have room in the funding array for this data
    (if (not (Utilities|Array|IsValidIndex (Variables|Default|GetCurrentFunds) (Variables|/CorruptionGameplay/Gameplay/WarEffort/Prototype/BP_StationProp_FundingStation.BP_StationProp_FundingStation_C|GetFundingStationNum _array_element)))
      (Utilities|Array|SetArrayElem (Variables|Default|GetCurrentFunds) (Variables|/CorruptionGameplay/Gameplay/WarEffort/Prototype/BP_StationProp_FundingStation.BP_StationProp_FundingStation_C|GetFundingStationNum _array_element) 0 true))
    (bind _asbp_station_prop_funding_station_multi_choice_new_weapon (Utilities|Casting|CastToBP_StationProp_FundingStation_MultiChoice_NewWeapon _array_element)
      (:then
        (Utilities|Array|SetArrayElem (Variables|Default|GetFundingPropNewWeapon) (Variables|/CorruptionGameplay/Gameplay/WarEffort/Prototype/BP_StationProp_FundingStation.BP_StationProp_FundingStation_C|GetFundingStationNum _asbp_station_prop_funding_station_multi_choice_new_weapon) _asbp_station_prop_funding_station_multi_choice_new_weapon true)
        ;[1 1 1] Set up the correct funding limit on the station
        (Variables|/CorruptionGameplay/Gameplay/WarEffort/Prototype/BP_StationProp_FundingStation.BP_StationProp_FundingStation_C|SetMaxFundsRequired (Variables|Default|GetChoiceNWFundingLimit) _asbp_station_prop_funding_station_multi_choice_new_weapon))
      (:CastFailed
        (bind _asbp_station_prop_funding_station_multi_choice_unvaults (Utilities|Casting|CastToBP_StationProp_FundingStation_MultiChoice_Unvaults _array_element)
          (:then
            (Utilities|Array|SetArrayElem (Variables|Default|GetFundingPropUnvault) (Variables|/CorruptionGameplay/Gameplay/WarEffort/Prototype/BP_StationProp_FundingStation.BP_StationProp_FundingStation_C|GetFundingStationNum _asbp_station_prop_funding_station_multi_choice_unvaults) _asbp_station_prop_funding_station_multi_choice_unvaults true)
            (Variables|/CorruptionGameplay/Gameplay/WarEffort/Prototype/BP_StationProp_FundingStation.BP_StationProp_FundingStation_C|SetMaxFundsRequired (Variables|Default|GetChoiceUVFundingLimit) _asbp_station_prop_funding_station_multi_choice_unvaults))
          (:CastFailed
            (bind _asbp_station_prop_funding_station_guard_tower (Utilities|Casting|CastToBP_StationProp_FundingStation_GuardTower _array_element)
              (:then
                (Variables|/CorruptionGameplay/Gameplay/WarEffort/Prototype/BP_StationProp_FundingStation.BP_StationProp_FundingStation_C|SetMaxFundsRequired (Variables|Default|GetTowerFundingLimit) _asbp_station_prop_funding_station_guard_tower))
              (:CastFailed
                (Variables|/CorruptionGameplay/Gameplay/WarEffort/Prototype/BP_StationProp_FundingStation.BP_StationProp_FundingStation_C|SetMaxFundsRequired (Variables|Default|GetFinalFundingLimit) _array_element)))))))))

(fn MultiChoiceCheck
  (params (InFunds :type Integer64) (InFundingStation :type Integer) (InFundingChoiceIndex :type Integer))
  (bind _fundingpropservice (Variables|Default|GetFundingPropService))
  (bind _output (Utilities|Array|Get(acopy) _fundingpropservice InFundingStation))
  (bind _output_2 (Utilities|ResolveSoftReference _output))
  (bind _currentactiveevent (Variables|/CorruptionGameplay/Gameplay/WarEffort/Prototype/BP_StationProp_FundingStation.BP_StationProp_FundingStation_C|GetCurrentActiveEvent _output_2))
  (Variables|Default|SetIn%20Funds InFunds)
  (Variables|Default|SetIn%20Funding%20Station InFundingStation)
  (Variables|Default|SetFunding%20Choice%20Index InFundingChoiceIndex)
  (Utilities|Casting|CastToBP_StationProp_FundingStation_MultiChoice_NewWeapon _output_2
    (:then
      (Variables|Default|SetIn%20Funding%20Station (select (> (Variables|Default|GetFunding%20Choice%20Index) 1) (+ (Variables|Default|GetSharedNewWeaponFunding) 1) (Variables|Default|GetSharedNewWeaponFunding))))
    (:CastFailed
      (Utilities|Casting|CastToBP_StationProp_FundingStation_MultiChoice_Unvaults _output_2
        (:then
          (Variables|Default|SetIn%20Funding%20Station (select (> (Variables|Default|GetFunding%20Choice%20Index) 1) (+ (Variables|Default|GetSharedUnvaultFunding) 1) (Variables|Default|GetSharedUnvaultFunding)))
          (LocalVariables|SetChoiceFund true)
          (if (Utilities|String|IsEmpty _currentactiveevent)
            ;[1 1 1] Backup calendar event check
            (for _array_element (Utilities|Array|MakeArray "WEUV01" "WENW01" "WEUV02" "WEUV03")
              (bind _biseventactive (CalendarEvents|GetCalendarEventInformation _array_element))
              (if _biseventactive
                (return (Variables|Default|GetIn%20Funding%20Station) (Variables|Default|GetIn%20Funds) (LocalVariables|GetChoiceFund) _array_element)))))
        (:CastFailed
          (LocalVariables|SetChoiceFund false)))
      (return (Variables|Default|GetIn%20Funding%20Station) (Variables|Default|GetIn%20Funds) (LocalVariables|GetChoiceFund) _currentactiveevent))))

(fn OnRep_TotalFunds
  (CallFunction|UpdateTotalFunds (Variables|TotalFunds|GetTotalFunds)))

(fn OnRep_UpdateRootInt
  (switch Utilities|FlowControl|Switch|SwitchonEMeshNetworkNodeType (MeshNetworking|GetMeshNetworkNodeType (Variables|Default|GetBP_FundingComponent))
    (:Edge
      (CallFunction|UpdateAllFundingStations))))

(fn SetAllFundsArray
  (params (Item :type Integer64%20%28by%20ref%29))
  (Utilities|Array|SetArrayElem (Variables|Default|GetCurrentFunds) (Variables|Default|GetFunding%20Station) Item true))

(fn SetRoundOptionDataFromCurrentFunds
  (Utilities|FlowControl|Sequence
    (:then_0
      (LocalVariables|SetNWFundA (Utilities|Array|Get(acopy) (Variables|Default|GetCurrentFunds) (Variables|Default|GetSharedNewWeaponFunding))))
    (:then_1
      (LocalVariables|SetNWFundB (Utilities|Array|Get(acopy) (Variables|Default|GetCurrentFunds) (+ (Variables|Default|GetSharedNewWeaponFunding) 1))))
    (:then_2
      (LocalVariables|SetUVFundA (Utilities|Array|Get(acopy) (Variables|Default|GetCurrentFunds) (Variables|Default|GetSharedUnvaultFunding))))
    (:then_3
      (LocalVariables|SetUVFundB (Utilities|Array|Get(acopy) (Variables|Default|GetCurrentFunds) (+ (Variables|Default|GetSharedUnvaultFunding) 1))))
    (:then_4
      ;[1 1 1] Keep the choice data in sync with the funding array for the current calendar event
      (for (_array_element _array_index) (Utilities|Array|MakeArray "WEUV01" "WENW01" "WEUV02" "WEUV03")
        (bind _biseventactive (CalendarEvents|GetCalendarEventInformation _array_element))
        (if _biseventactive
          (Fortnite|FortShippingLog self (Utilities|String|ToString(Text) (Utilities|Text|FormatText "NSLOCTEXT(\"[8FA25645435DB4DB77D0CFB0882B521F]\", \"C712185C45010E400A1E51803270A7DE\", \"Updating Option Data at index {0} for Calendar Event {1}. UVFundA: {2} UVFundB: {3} NWFundA: {4} NWFundB: {5}\")" _array_index _array_element (Math|Conversions|ToInteger(Integer64) (LocalVariables|GetUVFundA)) (Math|Conversions|ToInteger(Integer64) (LocalVariables|GetUVFundB)) (Math|Conversions|ToInteger(Integer64) (LocalVariables|GetNWFundA)) (Math|Conversions|ToInteger(Integer64) (LocalVariables|GetNWFundB)))))
          (switch int _array_index
            (:0
              (Utilities|Struct|SetmembersinWarEffortFundingChoiceData (Variables|Default|GetRound1OptionData) (Utilities|Struct|MakeWarEffortFundingOptionData "(TagName=\"Fort.Athena.WarEffort.Round1.Option1\")" (LocalVariables|GetUVFundA)) (Utilities|Struct|MakeWarEffortFundingOptionData "(TagName=\"Fort.Athena.WarEffort.Round1.Option2\")" (LocalVariables|GetUVFundB))))
            (:1
              (Utilities|Struct|SetmembersinWarEffortFundingChoiceData (Variables|Default|GetRound2OptionData) (Utilities|Struct|MakeWarEffortFundingOptionData "(TagName=\"Fort.Athena.WarEffort.Round2.Option1\")" (LocalVariables|GetNWFundA)) (Utilities|Struct|MakeWarEffortFundingOptionData "(TagName=\"Fort.Athena.WarEffort.Round2.Option2\")" (LocalVariables|GetNWFundB))))
            (:2
              (Utilities|Struct|SetmembersinWarEffortFundingChoiceData (Variables|Default|GetRound3OptionData) (Utilities|Struct|MakeWarEffortFundingOptionData "(TagName=\"Fort.Athena.WarEffort.Round3.Option1\")" (LocalVariables|GetUVFundA)) (Utilities|Struct|MakeWarEffortFundingOptionData "(TagName=\"Fort.Athena.WarEffort.Round3.Option2\")" (LocalVariables|GetUVFundB))))
            (:3
              (Utilities|Struct|SetmembersinWarEffortFundingChoiceData (Variables|Default|GetRound4OptionData) (Utilities|Struct|MakeWarEffortFundingOptionData "(TagName=\"Fort.Athena.WarEffort.Round4.Option1\")" (LocalVariables|GetUVFundA)) (Utilities|Struct|MakeWarEffortFundingOptionData "(TagName=\"Fort.Athena.WarEffort.Round4.Option2\")" (LocalVariables|GetUVFundB))))))))))

(fn UpdateTotalFunds
  (params (TotalFunds :type Integer64))
  (bind _bp_fundingcomponent (Variables|Default|GetBP_FundingComponent))
  (bind _returnvalue (MeshNetworking|GetMeshNetworkNodeType _bp_fundingcomponent))
  (bind _returnvalue_2 (Editor|InEditororPIE))
  (bind _newtotalfunds (Variables|Default|GetNewTotalFunds))
  (bind _returnvalue_3 (Utilities|Text|ToText(Integer64) _newtotalfunds))
  (bind _returnvalue_4 (Utilities|String|ToString(Text) _returnvalue_3))
  (bind _returnvalue_5 (Utilities|String|ToString(Integer) (Variables|Default|GetInitialFundingStation)))
  (bind _sharednewweaponfunding (Variables|Default|GetSharedNewWeaponFunding))
  (bind _funding_station (Variables|Default|GetFunding%20Station))
  (bind _returnvalue_6 (== _funding_station _sharednewweaponfunding))
  (bind _sharedunvaultfunding (Variables|Default|GetSharedUnvaultFunding))
  (bind _funding_station_2 (Variables|Default|GetFunding%20Station))
  (bind _returnvalue_7 (== _funding_station_2 _sharedunvaultfunding))
  (bind _initialfundingstation (Variables|Default|GetInitialFundingStation))
  (bind _fundingpropservice (Variables|Default|GetFundingPropService))
  (bind _output (Utilities|Array|Get(acopy) _fundingpropservice _initialfundingstation))
  (bind _output_2 (Utilities|ResolveSoftReference _output))
  (bind _returnvalue_8 (Utilities|GetDisplayName _output_2))
  (bind _debug_skipmeshnetwork (Variables|Debug|GetDebug_SkipMeshNetwork))
  (bind _returnvalue_9 (and _debug_skipmeshnetwork _returnvalue_2))
  (bind _returnvalue_10 (Utilities|String|Append _returnvalue_8 ":  " _returnvalue_4 "   FundingStation:  " _returnvalue_5))
  (bind _returnvalue_11 (or _returnvalue_6 _returnvalue_7))
  (bind _returnvalue_12 (select _returnvalue_9 "Edge" _returnvalue))
  (bind _returnvalue_13 (select _returnvalue_11 true false))
  (bind _initialfundingstation_2 (Variables|Default|GetInitialFundingStation))
  (bind _totalfunds (Variables|TotalFunds|GetTotalFunds))
  (if (> (Utilities|Array|Length (Variables|Default|GetFundingPropService)) 0)
    (switch Utilities|FlowControl|Switch|SwitchonEMeshNetworkNodeType _returnvalue_12
      (:Root
        ;[1 1 1] Test Only
        (if (Variables|Debug|GetDebug_PrintStrings)
          (Class|BPStationPropFundingStation|EdgeServerShowFunding (Utilities|ResolveSoftReference (Utilities|Array|Get(acopy) (Variables|Default|GetFundingPropService) (Variables|Default|GetFunding%20Station))) (Variables|TotalFunds|GetTotalFunds) false false)))
      (:Edge
        (if (Networking|IsDedicatedServer)
          (CallFunction|UpdateMutatorFundingData)
          (else
            (CallFunction|UpdateUIFundingData (Variables|Default|GetFunding%20Station) (Variables|Default|GetNewTotalFunds))))
        (bind _asbp_station_prop_funding_station_multi_choice (Utilities|Casting|CastToBP_StationProp_FundingStation_MultiChoice _output_2)
          (:then
            (if (>= (Variables|Default|GetNewTotalFunds) (Math|Float|TruncatetoInteger64 (ScalableFloat|GetValueatLevel (Variables|/CorruptionGameplay/Gameplay/WarEffort/Prototype/BP_StationProp_FundingStation.BP_StationProp_FundingStation_C|GetMaxFundsRequired _asbp_station_prop_funding_station_multi_choice))))
              (Utilities|Casting|CastToBP_StationProp_FundingStation_MultiChoice_NewWeapon _asbp_station_prop_funding_station_multi_choice
                (:then
                  (for _array_element (Variables|Default|GetFundingPropNewWeapon)
                    (Class|BPStationPropFundingStation|EdgeServerShowFunding _array_element (Variables|TotalFunds|GetTotalFunds) _returnvalue_13 false)))
                (:CastFailed
                  (Utilities|Casting|CastToBP_StationProp_FundingStation_MultiChoice_Unvaults _asbp_station_prop_funding_station_multi_choice)
                  (for _array_element_2 (Variables|Default|GetFundingPropUnvault)
                    (Class|BPStationPropFundingStation|EdgeServerShowFunding _array_element_2 (Variables|TotalFunds|GetTotalFunds) _returnvalue_13 false))))
              (else
                (Development|PrintString _returnvalue_10)
                (Class|BPStationPropFundingStation|EdgeServerShowFunding _output_2 _totalfunds (select (or (== (Variables|Default|GetFunding%20Station) (Variables|Default|GetSharedNewWeaponFunding)) (== (Variables|Default|GetFunding%20Station) (Variables|Default|GetSharedUnvaultFunding))) true false) false))))
          (:CastFailed
            (Development|PrintString _returnvalue_10)
            (Class|BPStationPropFundingStation|EdgeServerShowFunding _output_2 _totalfunds false true))))
      (:Client
        (CallFunction|UpdateUIFundingData (Variables|Default|GetFunding%20Station) (Variables|Default|GetNewTotalFunds))))
    (else
      (switch Utilities|FlowControl|Switch|SwitchonEMeshNetworkNodeType _returnvalue_12
        (:Edge
          ;[1 1 1] Client servers wait for the match to start before getting the funding services; don't try anything before then!
          (Fortnite|FortShippingLog self (Utilities|String|ToString(Text) (Utilities|Text|FormatText "NSLOCTEXT(\"[8FA25645435DB4DB77D0CFB0882B521F]\", \"67412E76483C6AE19134739C6D01DD75\", \"Funding Prop Service array not set up! skipping UpdateTotalFunds\")"))))))))

(fn UpdateUIFundingData
  (params (FundingStationIndex :type Integer) (NewCurrentFund :type Integer64))
  (if (not (Networking|IsDedicatedServer))
    (CallFunction|BroadcaseUpdatedFundsMessage (Utilities|Struct|MakeItemFundingUpdatedFundingData FundingStationIndex NewCurrentFund))))

(fn UpdateAllFundingStations
  (bind _sharedunvaultfunding (Variables|Default|GetSharedUnvaultFunding))
  (bind _returnvalue (+ _sharedunvaultfunding 1))
  (bind _sharednewweaponfunding (Variables|Default|GetSharedNewWeaponFunding))
  (bind _returnvalue_2 (+ _sharednewweaponfunding 1))
  (bind _currentfunds (Variables|Default|GetCurrentFunds))
  (bind _output (Utilities|Array|Get(acopy) _currentfunds _returnvalue))
  (bind _sharedunvaultfunding_2 (Variables|Default|GetSharedUnvaultFunding))
  (bind _currentfunds_2 (Variables|Default|GetCurrentFunds))
  (bind _output_2 (Utilities|Array|Get(acopy) _currentfunds_2 _sharedunvaultfunding_2))
  (bind _sharednewweaponfunding_2 (Variables|Default|GetSharedNewWeaponFunding))
  (bind _currentfunds_3 (Variables|Default|GetCurrentFunds))
  (bind _output_3 (Utilities|Array|Get(acopy) _currentfunds_3 _sharednewweaponfunding_2))
  (bind _currentfunds_4 (Variables|Default|GetCurrentFunds))
  (bind _output_4 (Utilities|Array|Get(acopy) _currentfunds_4 _returnvalue_2))
  (bind _currentfunds_5 (Variables|Default|GetCurrentFunds))
  ;[1 1 1] Push updates to all of the funding stations based on the current data
  (if (> (Utilities|Array|Length (Variables|Default|GetFundingPropService)) 0)
    (for (_array_element _array_index) (Variables|Default|GetFundingPropService)
      (Utilities|IsValid (Utilities|ResolveSoftReference _array_element)
        (:Is%20Valid
          (bind _asbp_station_prop_funding_station_multi_choice_new_weapon (Utilities|Casting|CastToBP_StationProp_FundingStation_MultiChoice_NewWeapon (Utilities|ResolveSoftReference _array_element))
            (:then
              (if (>= _output_3 (Math|Float|TruncatetoInteger64 (ScalableFloat|GetValueatLevel (Variables|Default|GetChoiceNWFundingLimit))))
                (Class|BPStationPropFundingStation|FundedStateA _asbp_station_prop_funding_station_multi_choice_new_weapon)
                (else
                  (Class|BPStationPropFundingStation|EdgeServerShowFunding _asbp_station_prop_funding_station_multi_choice_new_weapon _output_3 true false)
                  (if (>= _output_4 (Math|Float|TruncatetoInteger64 (ScalableFloat|GetValueatLevel (Variables|Default|GetChoiceNWFundingLimit))))
                    (Class|BPStationPropFundingStation|FundedStateB _asbp_station_prop_funding_station_multi_choice_new_weapon)
                    (else
                      (Class|BPStationPropFundingStation|EdgeServerShowFunding _asbp_station_prop_funding_station_multi_choice_new_weapon _output_4 false false))))))
            (:CastFailed
              (bind _asbp_station_prop_funding_station_multi_choice_unvaults (Utilities|Casting|CastToBP_StationProp_FundingStation_MultiChoice_Unvaults (Utilities|ResolveSoftReference _array_element))
                (:then
                  (if (>= _output_2 (Math|Float|TruncatetoInteger64 (ScalableFloat|GetValueatLevel (Variables|Default|GetChoiceUVFundingLimit))))
                    (Class|BPStationPropFundingStation|FundedStateA _asbp_station_prop_funding_station_multi_choice_unvaults)
                    (else
                      (Class|BPStationPropFundingStation|EdgeServerShowFunding _asbp_station_prop_funding_station_multi_choice_unvaults _output_2 true false)
                      (if (>= _output (Math|Float|TruncatetoInteger64 (ScalableFloat|GetValueatLevel (Variables|Default|GetChoiceUVFundingLimit))))
                        (Class|BPStationPropFundingStation|FundedStateB _asbp_station_prop_funding_station_multi_choice_unvaults)
                        (else
                          (Class|BPStationPropFundingStation|EdgeServerShowFunding _asbp_station_prop_funding_station_multi_choice_unvaults _output false false))))))
                (:CastFailed
                  (bind _asbp_station_prop_funding_station_guard_tower (Utilities|Casting|CastToBP_StationProp_FundingStation_GuardTower (Utilities|ResolveSoftReference _array_element))
                    (:then
                      (if (>= (Utilities|Array|Get(acopy) _currentfunds_5 _array_index) (Math|Float|TruncatetoInteger64 (ScalableFloat|GetValueatLevel (Variables|Default|GetTowerFundingLimit))))
                        (Class|BPStationPropFundingStation|FundedState _asbp_station_prop_funding_station_guard_tower)
                        (else
                          (Class|BPStationPropFundingStation|EdgeServerShowFunding _asbp_station_prop_funding_station_guard_tower (Utilities|Array|Get(acopy) _currentfunds_5 _array_index) false true))))
                    (:CastFailed
                      (bind _asbp_station_prop_funding_station_other (Utilities|Casting|CastToBP_StationProp_FundingStation_Other (Utilities|ResolveSoftReference _array_element)))
                      (if (>= (Utilities|Array|Get(acopy) _currentfunds_5 _array_index) (Math|Float|TruncatetoInteger64 (ScalableFloat|GetValueatLevel (Variables|Default|GetFinalFundingLimit))))
                        (Class|BPStationPropFundingStation|FundedState _asbp_station_prop_funding_station_other)
                        (else
                          (Class|BPStationPropFundingStation|EdgeServerShowFunding _asbp_station_prop_funding_station_other (Utilities|Array|Get(acopy) _currentfunds_5 _array_index) false true))))))))))))
    (else
      ;;[1 1 1] Client servers wait for the match to start before getting the funding services; don't try anything before then!
      (Fortnite|FortShippingLog self (Utilities|String|ToString(Text) (Utilities|Text|FormatText "NSLOCTEXT(\"[8FA25645435DB4DB77D0CFB0882B521F]\", \"67412E76483C6AE19134739C6D01DD75\", \"Funding Prop Service array not set up! skipping UpdateTotalFunds\")"))))))

(fn UpdateMutatorFundingData
  (Utilities|FlowControl|Sequence
    (:then_0
      (if (Class|ItemUIFunctionLibrary|IsNullItem (Variables|Default|GetWarEffortMutatorRef))
        (bind _asfort_game_state_athena (Utilities|Casting|CastToFortGameStateAthena (Game|GetGameState)))
        (if (not (Class|ItemUIFunctionLibrary|IsNullItem (Modifier|GetMutatorbyClass (Variables|/Script/FortniteGame.FortGameStateAthena|GetMutatorListComponent _asfort_game_state_athena) "/Script/CorruptionGameplayCodeRuntime.FortAthenaMutator_WarEffort")))
          (bind _asfort_athena_mutator_war_effort (Utilities|Casting|CastToFortAthenaMutator_WarEffort (Modifier|GetMutatorbyClass (Variables|/Script/FortniteGame.FortGameStateAthena|GetMutatorListComponent _asfort_game_state_athena) "/Script/CorruptionGameplayCodeRuntime.FortAthenaMutator_WarEffort")))
          (Variables|Default|SetWarEffortMutatorRef _asfort_athena_mutator_war_effort))))
    (:then_1
      (if (not (Class|ItemUIFunctionLibrary|IsNullItem (Variables|Default|GetWarEffortMutatorRef)))
        (Utilities|FlowControl|Sequence
          (:then_0
            (LocalVariables|SetUVFundingOptionA (Utilities|Array|Get(acopy) (Variables|Default|GetCurrentFunds) (Variables|Default|GetSharedUnvaultFunding))))
          (:then_1
            (LocalVariables|SetUVFundingOptionB (Utilities|Array|Get(acopy) (Variables|Default|GetCurrentFunds) (+ (Variables|Default|GetSharedUnvaultFunding) 1))))
          (:then_2
            (LocalVariables|SetNWFundingOptionA (Utilities|Array|Get(acopy) (Variables|Default|GetCurrentFunds) (Variables|Default|GetSharedNewWeaponFunding))))
          (:then_3
            (LocalVariables|SetNWFundingOptionB (Utilities|Array|Get(acopy) (Variables|Default|GetCurrentFunds) (+ (Variables|Default|GetSharedNewWeaponFunding) 1))))
          (:then_4
            (Utilities|FlowControl|Sequence
              (:then_0
                (bind _biseventactive (CalendarEvents|GetCalendarEventInformation "WEUV01"))
                (if (or _biseventactive (Editor|InEditororPIE))
                  (Class|FortAthenaMutatorWarEffort|SetItemFundedAmount (Variables|Default|GetWarEffortMutatorRef) "(TagName=\"Fort.Athena.WarEffort.Round1.Option1\")" (LocalVariables|GetUVFundingOptionA) (Math|Float|TruncatetoInteger64 (ScalableFloat|GetValueatLevel (Variables|Default|GetChoiceUVFundingLimit))))
                  (Class|FortAthenaMutatorWarEffort|SetItemFundedAmount (Variables|Default|GetWarEffortMutatorRef) "(TagName=\"Fort.Athena.WarEffort.Round1.Option2\")" (LocalVariables|GetUVFundingOptionB) (Math|Float|TruncatetoInteger64 (ScalableFloat|GetValueatLevel (Variables|Default|GetChoiceUVFundingLimit))))))
              (:then_1
                (bind _biseventactive_2 (CalendarEvents|GetCalendarEventInformation "WENW01"))
                (if (or _biseventactive_2 (Editor|InEditororPIE))
                  (Class|FortAthenaMutatorWarEffort|SetItemFundedAmount (Variables|Default|GetWarEffortMutatorRef) "(TagName=\"Fort.Athena.WarEffort.Round2.Option1\")" (LocalVariables|GetNWFundingOptionA) (Math|Float|TruncatetoInteger64 (ScalableFloat|GetValueatLevel (Variables|Default|GetChoiceNWFundingLimit))))
                  (Class|FortAthenaMutatorWarEffort|SetItemFundedAmount (Variables|Default|GetWarEffortMutatorRef) "(TagName=\"Fort.Athena.WarEffort.Round2.Option2\")" (LocalVariables|GetNWFundingOptionB) (Math|Float|TruncatetoInteger64 (ScalableFloat|GetValueatLevel (Variables|Default|GetChoiceNWFundingLimit))))))
              (:then_2
                (bind _biseventactive_3 (CalendarEvents|GetCalendarEventInformation "WEUV02"))
                (if _biseventactive_3
                  (Class|FortAthenaMutatorWarEffort|SetItemFundedAmount (Variables|Default|GetWarEffortMutatorRef) "(TagName=\"Fort.Athena.WarEffort.Round3.Option1\")" (LocalVariables|GetUVFundingOptionA) (Math|Float|TruncatetoInteger64 (ScalableFloat|GetValueatLevel (Variables|Default|GetChoiceUVFundingLimit))))
                  (Class|FortAthenaMutatorWarEffort|SetItemFundedAmount (Variables|Default|GetWarEffortMutatorRef) "(TagName=\"Fort.Athena.WarEffort.Round3.Option2\")" (LocalVariables|GetUVFundingOptionB) (Math|Float|TruncatetoInteger64 (ScalableFloat|GetValueatLevel (Variables|Default|GetChoiceUVFundingLimit))))))
              (:then_3
                (bind _biseventactive_4 (CalendarEvents|GetCalendarEventInformation "WEUV03"))
                (if _biseventactive_4
                  (Class|FortAthenaMutatorWarEffort|SetItemFundedAmount (Variables|Default|GetWarEffortMutatorRef) "(TagName=\"Fort.Athena.WarEffort.Round4.Option1\")" (LocalVariables|GetUVFundingOptionA) (Math|Float|TruncatetoInteger64 (ScalableFloat|GetValueatLevel (Variables|Default|GetChoiceUVFundingLimit))))
                  (Class|FortAthenaMutatorWarEffort|SetItemFundedAmount (Variables|Default|GetWarEffortMutatorRef) "(TagName=\"Fort.Athena.WarEffort.Round4.Option2\")" (LocalVariables|GetUVFundingOptionB) (Math|Float|TruncatetoInteger64 (ScalableFloat|GetValueatLevel (Variables|Default|GetChoiceUVFundingLimit)))))))))))))

(fn ConstructionScript
  (bind _choiceuvfundinglimit (Variables|Default|GetChoiceUVFundingLimit))
  (bind _returnvalue (ScalableFloat|GetValueatLevel _choiceuvfundinglimit))
  (bind _returnvalue_2 (Math|Float|TruncatetoInteger64 _returnvalue))
  (|CallParentFunction)
  (Utilities|FlowControl|Sequence
    (:then_0
      (Variables|SetRound1OptionData_Option1_OptionTag (Utilities|Struct|MakeWarEffortFundingChoiceData (struct :OptionTag "(TagName=\"Fort.Athena.WarEffort.Round1.Option1\")" :CurrentFundingAmount 0) (struct :OptionTag "(TagName=\"Fort.Athena.WarEffort.Round1.Option2\")" :CurrentFundingAmount 0) _returnvalue_2)))
    (:then_1
      (Variables|SetRound2OptionData_Option1_OptionTag (Utilities|Struct|MakeWarEffortFundingChoiceData (struct :OptionTag "(TagName=\"Fort.Athena.WarEffort.Round2.Option1\")" :CurrentFundingAmount 0) (struct :OptionTag "(TagName=\"Fort.Athena.WarEffort.Round2.Option2\")" :CurrentFundingAmount 0) (Math|Float|TruncatetoInteger64 (ScalableFloat|GetValueatLevel (Variables|Default|GetChoiceNWFundingLimit))))))
    (:then_2
      (Variables|SetRound3OptionData_Option1_OptionTag (Utilities|Struct|MakeWarEffortFundingChoiceData (struct :OptionTag "(TagName=\"Fort.Athena.WarEffort.Round3.Option1\")" :CurrentFundingAmount 0) (struct :OptionTag "(TagName=\"Fort.Athena.WarEffort.Round3.Option2\")" :CurrentFundingAmount 0) _returnvalue_2)))
    (:then_3
      (Variables|SetRound4OptionData_Option1_OptionTag (Utilities|Struct|MakeWarEffortFundingChoiceData (struct :OptionTag "(TagName=\"Fort.Athena.WarEffort.Round4.Option1\")" :CurrentFundingAmount 0) (struct :OptionTag "(TagName=\"Fort.Athena.WarEffort.Round4.Option2\")" :CurrentFundingAmount 0) _returnvalue_2)))))
